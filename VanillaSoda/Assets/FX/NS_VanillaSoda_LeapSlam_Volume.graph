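# Niagara system for a VanillaSoda leap-slam impact effect, the "Volume" version by name.
# Ten emitters layer debris, sparks, dust, streaks, a light, shockwave, two mesh flashes and two decals, judging by their names.
# User parameters expose decal offset, two non-uniform decal scales and wave scale.
# System scripts use SystemState and a spawn-time variable assignment, with color and random-float dynamic inputs available; renderer details are absent.
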NiagaraSystem /VanillaSoda/Assets/FX/NS_VanillaSoda_LeapSlam_Volume
dependencies {
 usedDataInterfaces []
 usedDynamicInputs [&MakeLinearColor, &RandomRangeFloat]
 usedModules [
  &self:SystemSpawnScript.SystemScriptSource.SystemScriptGraph.NiagaraNodeAssignment_0.SetVariables_#1
  &SystemState
 ]
}
emitters {
 Decal {rendererClasses [], renderers []}
 Decal001 {rendererClasses [], renderers []}
 Dust {rendererClasses [], renderers []}
 FlashMesh {rendererClasses [], renderers []}
 FlashMeshWide {rendererClasses [], renderers []}
 Light {rendererClasses [], renderers []}
 ShockwaveL {rendererClasses [], renderers []}
 Sparks {rendererClasses [], renderers []}
 Streaks {rendererClasses [], renderers []}
 debris {rendererClasses [], renderers []}
}
summary {
 systemName=NS_VanillaSoda_LeapSlam_Volume
 emitters [
  {emitterName=debris, rendererClasses []}
  {emitterName=Sparks, rendererClasses []}
  {emitterName=Light, rendererClasses []}
  {emitterName=ShockwaveL, rendererClasses []}
  {emitterName=FlashMesh, rendererClasses []}
  {emitterName=Dust, rendererClasses []}
  {emitterName=FlashMeshWide, rendererClasses []}
  {emitterName=Streaks, rendererClasses []}
  {emitterName=Decal, rendererClasses []}
  {emitterName=Decal001, rendererClasses []}
 ]
 userVariables [
  User.Decal Ofs {type=NiagaraPosition, description="", defaultValue {}}
  User.Decal Scale {type=Vector3f, description="", defaultValue {struct=Vector3f, value=<150 1000 550>}}
  User.DEcal Scale2 {type=Vector3f, description="", defaultValue {struct=Vector3f, value=<150 600 75>}}
  User.WaveScale {type=Vector3f, description="", defaultValue {struct=Vector3f, value=<15 25 30>}}
 ]
}
