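# Niagara system for VanillaSoda that, by its name, plays the effect for a completed pull-out action.
# Eight emitters cover debris, dust, sparks, two stone layers, a wide flash mesh, a light and a decal.
# System scripts use SystemState, a spawn-time variable assignment and the FNTimeOfDayColor dynamic input, suggesting time-of-day tinting.
# It exposes no user variables, and no renderer details are listed.
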NiagaraSystem /VanillaSoda/Assets/FX/NS_VanillaSoda_PullOut_Done
dependencies {
 usedDataInterfaces []
 usedDynamicInputs [&FNTimeOfDayColor]
 usedModules [
  &self:SystemSpawnScript.SystemScriptSource.SystemScriptGraph.NiagaraNodeAssignment_0.SetVariables_#1
  &SystemState
 ]
}
emitters {
 Decal {rendererClasses [], renderers []}
 Dust {rendererClasses [], renderers []}
 FlashMeshWide {rendererClasses [], renderers []}
 Light {rendererClasses [], renderers []}
 Sparks {rendererClasses [], renderers []}
 Stone {rendererClasses [], renderers []}
 Stone001 {rendererClasses [], renderers []}
 debris {rendererClasses [], renderers []}
}
summary {
 systemName=NS_VanillaSoda_PullOut_Done
 emitters [
  {emitterName=debris, rendererClasses []}
  {emitterName=Sparks, rendererClasses []}
  {emitterName=Light, rendererClasses []}
  {emitterName=Dust, rendererClasses []}
  {emitterName=FlashMeshWide, rendererClasses []}
  {emitterName=Stone, rendererClasses []}
  {emitterName=Stone001, rendererClasses []}
  {emitterName=Decal, rendererClasses []}
 ]
 userVariables []
}
