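# NS_VanillaSoda_Pull_SwordOnly is a Niagara effect system, named as the sword-only variant of a VanillaSoda pull effect.
# Its twelve emitters are named for two sword layers (SwordG, SwordR), electric arcs, ring and bolts, light rays, lines, floor gather, cracks and a looping light flicker.
# System State is the only listed module.
# No user variables, renderers or data interfaces are listed.
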NiagaraSystem /VanillaSoda/Assets/FX/NS_VanillaSoda_Pull_SwordOnly
dependencies {usedDataInterfaces [], usedDynamicInputs [], usedModules [&SystemState]}
emitters {
 Cracks {rendererClasses [], renderers []}
 Default_LightAndGlowFlicker_Loop {rendererClasses [], renderers []}
 ElectricArcsLarge001 {rendererClasses [], renderers []}
 ElectricRing {rendererClasses [], renderers []}
 FloorGather {rendererClasses [], renderers []}
 Light_Rays {rendererClasses [], renderers []}
 Lightning_Bolts {rendererClasses [], renderers []}
 Lines {rendererClasses [], renderers []}
 Lines001 {rendererClasses [], renderers []}
 RaysBase {rendererClasses [], renderers []}
 SwordG {rendererClasses [], renderers []}
 SwordR {rendererClasses [], renderers []}
}
summary {
 systemName=NS_VanillaSoda_Pull_SwordOnly
 emitters [
  {emitterName=SwordG, rendererClasses []}
  {emitterName=SwordR, rendererClasses []}
  {emitterName=ElectricArcsLarge001, rendererClasses []}
  {emitterName=Light_Rays, rendererClasses []}
  {emitterName=Lines001, rendererClasses []}
  {emitterName=Lines, rendererClasses []}
  {emitterName=Lightning_Bolts, rendererClasses []}
  {emitterName=ElectricRing, rendererClasses []}
  {emitterName=RaysBase, rendererClasses []}
  {emitterName=FloorGather, rendererClasses []}
  {emitterName=Cracks, rendererClasses []}
  {emitterName=Default_LightAndGlowFlicker_Loop, rendererClasses []}
 ]
 userVariables []
}
